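# Physics asset for the female medium Prison Break face accessory, covering jiggle bones for two ears and a hood zipper around a head capsule.
# Each ear chains from head through a small master body to a larger capsule with narrow swing limits and zero twist.
# The zipper hangs off the two-bone right hood corner chain via its own master, with wider swing limits.
# Links from head through hood corners and masters are free in swing and twist, leaving only tip joints angularly constrained.
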PhysicsAsset /Game/Characters/Player/Female/Medium/Bodies/F_MED_Prison_Break/Meshes/Parts/F_MED_Prison_Break_FaceAcc_Physics
bodies {
 dyn_ear_l_01 [{shapeType=Capsule, length=2.7204, radius=2.2796, shapeName=dyn_ear_r_01_capsule, center=<1.7 0 0>, rotation=<rot 90 90 0>}]
 dyn_ear_l_MSTR [{shapeType=Capsule, length=0.5, radius=0.5, shapeName=dyn_ear_l_MSTR_capsule}]
 dyn_ear_r_01 [{shapeType=Capsule, length=2.7204, radius=2.2796, shapeName=dyn_ear_r_01_capsule, center=<1.7 0 0>, rotation=<rot 90 90 0>}]
 dyn_ear_r_MSTR [{shapeType=Capsule, length=0.5, radius=0.5, shapeName=dyn_ear_r_MSTR_capsule}]
 dyn_hoodCrnr_r_01 [{shapeType=Capsule, length=0.5, radius=0.5, shapeName=dyn_hoodCrnr_r_01_capsule}]
 dyn_hoodCrnr_r_02 [{shapeType=Capsule, length=0.5, radius=0.5, shapeName=dyn_hoodCrnr_r_02_capsule}]
 dyn_zipperFaceAcc_01 [{shapeType=Capsule, length=4.3695, radius=1.5258, shapeName=dyn_zipperFaceAcc_01_capsule, center=<3.2483 0 -0.8>, rotation=<rot 90 0 0>}]
 dyn_zipperFaceAcc_MSTR [{shapeType=Capsule, length=0.5, radius=0.5, shapeName=dyn_zipperFaceAcc_MSTR_capsule}]
 head [{shapeType=Capsule, length=13.4485, radius=12.2694, shapeName=head_capsule, center=<-0.0948 -0.505 -1.4515>, rotation=<rot 86.2646 -0.0001 0>}]
}
constraints [
 {bone1Name=dyn_zipperFaceAcc_01, bone2Name=dyn_zipperFaceAcc_MSTR, swing1LimitDegrees=45, swing2LimitDegrees=60, twistLimitDegrees=0}
 {
  bone1Name=dyn_ear_l_MSTR
  bone2Name=head
  swing1LimitDegrees=45
  swing1Motion=Free
  swing2LimitDegrees=45
  swing2Motion=Free
  twistLimitDegrees=45
  twistMotion=Free
 }
 {bone1Name=dyn_ear_l_01, bone2Name=dyn_ear_l_MSTR, swing1LimitDegrees=8, swing2LimitDegrees=20, twistLimitDegrees=0}
 {
  bone1Name=dyn_ear_r_MSTR
  bone2Name=head
  swing1LimitDegrees=45
  swing1Motion=Free
  swing2LimitDegrees=45
  swing2Motion=Free
  twistLimitDegrees=45
  twistMotion=Free
 }
 {bone1Name=dyn_ear_r_01, bone2Name=dyn_ear_r_MSTR, swing1LimitDegrees=8, swing2LimitDegrees=20, twistLimitDegrees=0}
 {
  bone1Name=dyn_zipperFaceAcc_MSTR
  bone2Name=dyn_hoodCrnr_r_02
  swing1LimitDegrees=45
  swing1Motion=Free
  swing2LimitDegrees=45
  swing2Motion=Free
  twistLimitDegrees=45
  twistMotion=Free
 }
 {
  bone1Name=dyn_hoodCrnr_r_01
  bone2Name=head
  swing1LimitDegrees=45
  swing1Motion=Free
  swing2LimitDegrees=45
  swing2Motion=Free
  twistLimitDegrees=45
  twistMotion=Free
 }
 {
  bone1Name=dyn_hoodCrnr_r_02
  bone2Name=dyn_hoodCrnr_r_01
  swing1LimitDegrees=45
  swing1Motion=Free
  swing2LimitDegrees=45
  swing2Motion=Free
  twistLimitDegrees=45
  twistMotion=Free
 }
]
