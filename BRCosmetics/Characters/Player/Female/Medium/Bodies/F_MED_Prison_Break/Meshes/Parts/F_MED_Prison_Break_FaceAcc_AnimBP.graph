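# Animation Blueprint for the F_MED_Prison_Break face accessory, driving its dynamic hood bones.
# It copies the attached parent's pose, runs head-relative RigidBody simulation, then additively offsets the bottom-center hood bone by a head Rot Back alpha.
# Each update derives head twist and back-rotation alphas from the body mesh's head-to-spine_05 delta.
# The twist-driven bone chain is disconnected from the output, and the connected left-corner modifier has zero offsets.
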AnimBlueprint /BRCosmetics/Characters/Player/Female/Medium/Bodies/F_MED_Prison_Break/Meshes/Parts/F_MED_Prison_Break_FaceAcc_AnimBP
target_skeleton=/Game/Characters/Player/Female/Medium/Bodies/F_MED_Prison_Break/Meshes/Parts/F_MED_Prison_Break_FaceAcc_Skeleton
graphs [
 AnimGraph {nodes [
   n1 = Get head Twist Pos() -> (head Twist Pos:float)
   n2 = "Transform (Modify) Bone - Bone: dyn_hoodBottom_c_01"(Alpha=@n1, Translation="0,0,0", Rotation="-10,0,5", Scale="1,1,1") {
    rotationMode=BMM_Additive
    rotationSpace=BCS_BoneSpace
    alphaBoolBlend {blendOption=Linear}
    alphaScaleBiasClamp {bInterpResult=true}
    boneToModify {boneName=dyn_hoodBottom_c_01}
    rotation {pitch=-10, roll=5}
   }
   "Transform (Modify) Bone - Bone: dyn_hoodCrnr_l_01"(ComponentPose=@n2, Alpha=@n1, Translation="0,0,0", Rotation="0,0,0", Scale="1,1,1") {
    rotationMode=BMM_Additive
    rotationSpace=BCS_BoneSpace
    alphaBoolBlend {blendOption=Linear}
    alphaScaleBiasClamp {bInterpResult=true}
    boneToModify {boneName=dyn_hoodCrnr_l_01}
   }
   n3 = RigidBody(ComponentPose=Local To Component(LocalPose=Copy Pose From Mesh() {bUseAttachedParent=true})) {
    bClampLinearTranslationLimitToRefPose=true
    bOverrideWorldGravity=true
    overridePhysicsAsset=/Game/Characters/Player/Female/Medium/Bodies/F_MED_Prison_Break/Meshes/Parts/F_MED_Prison_Break_FaceAcc_Physics
    simulationSpace=BaseBoneSpace
    alphaBoolBlend {blendOption=Linear}
    baseBoneRef {boneName=head}
    componentAppliedLinearAccClamp=<6000 6000 6000>
    componentLinearAccScale=<0.5 0.4 0.4>
    componentLinearVelScale=<0.5 0.4 0.4>
    overrideWorldGravity {z=-980}
   }
   n4 = "Transform (Modify) Bone - Bone: dyn_hoodBottom_c_01"(ComponentPose=@n3, Alpha=Get head Rot Back() -> (head Rot Back:float), Translation="0,2,0", Rotation="0,15,0", Scale="1,1,1") {
    rotationMode=BMM_Additive
    rotationSpace=BCS_BoneSpace
    translationMode=BMM_Additive
    translationSpace=BCS_BoneSpace
    alphaBoolBlend {blendOption=Linear}
    alphaScaleBiasClamp {bMapRange=true}
    boneToModify {boneName=dyn_hoodBottom_c_01}
    rotation {yaw=15}
    translation {y=2}
   }
   n5 = "Transform (Modify) Bone - Bone: dyn_hoodCrnr_l_01"(ComponentPose=@n4, Alpha=1, Translation="0,0,0", Rotation="0,0,0", Scale="1,1,1") {
    rotationMode=BMM_Additive
    rotationSpace=BCS_BoneSpace
    translationMode=BMM_Additive
    translationSpace=BCS_BoneSpace
    alphaBoolBlend {blendOption=Linear}
    alphaScaleBiasClamp {bMapRange=true}
    boneToModify {boneName=dyn_hoodCrnr_l_01}
   }
   Output Pose(Result=Component To Local(ComponentPose=@n5)) {layerGroup=None}
  ]}
 EventGraph {nodes [
   Event BlueprintUpdateAnimation()
   > n1 = Sequence!(execute) -> (then_0, then_1, then_2:Exec, then_3:Exec)
   > Set bodyMesh(bodyMesh:Skeletal Mesh Component Object Reference=GetPartSkeletalMeshComponent()) -> (Output_Get:Skeletal Mesh Component Object Reference)
   Set head Twist Pos(
    execute=@n1.then_1
    head Twist Pos:float=GetAlphaFromReferenceTransformDelta(
     SkeletalMeshComponent=Get bodyMesh() -> (bodyMesh:Skeletal Mesh Component Object Reference)
     BoneName=head
     BaseName=spine_05
     DeltaComponentType=QuaternionTwist
     InputRange=(X=10,Y=55)
     TwistAxis="1,0,0"
    )
   ) -> (Output_Get:float)
   Set head Rot Back(
    execute=@n1.then_3
    head Rot Back:float=GetAlphaFromReferenceTransformDelta(SkeletalMeshComponent=Get bodyMesh(), BoneName=head, BaseName=spine_05, InputRange=(X=0,Y=-40), TwistAxis="1,0,0")
   ) -> (Output_Get:float)
   Set head Twist Neg(
    execute=@n1.then_2
    head Twist Neg:float=GetAlphaFromReferenceTransformDelta(
     SkeletalMeshComponent=Get bodyMesh()
     BoneName=head
     BaseName=spine_05
     DeltaComponentType=QuaternionTwist
     InputRange=(X=-10,Y=-55)
     TwistAxis="1,0,0"
    )
   ) -> (Output_Get:float)
  ]}
]
variables [
 bodyMesh {type=SkeletalMeshComponent}
 head Rot Back {type=number}
 head Twist Neg {type=number}
 head Twist Pos {type=number}
]
