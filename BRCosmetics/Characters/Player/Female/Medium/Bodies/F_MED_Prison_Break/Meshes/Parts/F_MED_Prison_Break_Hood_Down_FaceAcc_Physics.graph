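# Physics asset for the hood-down face accessory of the F_MED_Prison_Break female cosmetic, defining capsule bodies for dynamic hood, ear and hair bones.
# Head, clavicle and upper-arm capsules approximate the character's upper body.
# Back, back-top, front and hair bones pair with MSTR roots under tight swing limits and zero twist limit; ears hang off the back bone.
# Hair roots-to-head and upperarm-to-clavicle constraints leave swing and twist free.
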PhysicsAsset /BRCosmetics/Characters/Player/Female/Medium/Bodies/F_MED_Prison_Break/Meshes/Parts/F_MED_Prison_Break_Hood_Down_FaceAcc_Physics
bodies {
 clavicle_l [{shapeType=Capsule, length=13.25, radius=7, shapeName=clavicle_l_capsule, center=<0 -1.6991 0>, rotation=<rot 90 0 0>}]
 clavicle_r [{shapeType=Capsule, length=13.25, radius=7, shapeName=clavicle_l_capsule, center=<0 1.6991 0>, rotation=<rot 90 0 180>}]
 dyn_hoodDownBackTop_c_01 [{
   shapeType=Capsule
   length=6.25
   radius=3.25
   shapeName=dyn_hoodDownBackTop_c_01_capsule
   center=<14.7818 -6 0>
   rotation=<rot 90 -14.0362 32.5538>
  }]
 dyn_hoodDownBackTop_c_MSTR [{shapeType=Capsule, length=2, radius=2, shapeName=dyn_hoodDownBackTop_c_MSTR_capsule}]
 dyn_hoodDownBack_c_01 [{shapeType=Capsule, length=5.75, radius=2.75, center=<8.4827 0.0448 0>, rotation=<rot 90 0 0>}]
 dyn_hoodDownBack_c_MSTR [{shapeType=Capsule, length=2, radius=2, shapeName=dyn_hoodDownBack_c_MSTR_capsule}]
 dyn_hoodDownEar_l_MSTR [{shapeType=Capsule, length=1.75, radius=1.75, shapeName=dyn_hoodDownEar_l_MSTR_capsule, center=<2.7518 0 0>, rotation=<rot 90 0 0>}]
 dyn_hoodDownEar_r_MSTR [{shapeType=Capsule, length=1.75, radius=1.75, shapeName=dyn_hoodDownEar_r_MSTR_capsule, center=<2.7518 0 0>, rotation=<rot 90 0 0>}]
 dyn_hoodDownFront_c_01 [{shapeType=Capsule, length=5.5, radius=2.5, shapeName=dyn_hoodDownFront_c_01_capsule, center=<8.0093 0.0383 0>, rotation=<rot 90 0 0>}]
 dyn_hoodDownFront_c_MSTR [{shapeType=Capsule, length=2, radius=2, shapeName=dyn_hoodDownFront_c_MSTR_capsule}]
 dyn_hoodDownHairA_l_01 [{shapeType=Capsule, length=2.75, radius=1.25, shapeName=dyn_hoodDownHairA_l_01_capsule, center=<2.7126 0 0>, rotation=<rot 90 0 0>}]
 dyn_hoodDownHairA_l_MSTR [{shapeType=Capsule, length=0.5, radius=0.5, shapeName=dyn_hoodDownHairA_l_MSTR_capsule}]
 dyn_hoodDownHairA_r_01 [{shapeType=Capsule, length=2.75, radius=1.25, shapeName=dyn_hoodDownHairA_r_01_capsule, center=<2.7126 0 0>, rotation=<rot 90 0 0>}]
 dyn_hoodDownHairA_r_MSTR [{shapeType=Capsule, length=0.5, radius=0.5, shapeName=dyn_hoodDownHairA_r_MSTR_capsule}]
 dyn_hoodDownHair_l_01 [{shapeType=Capsule, length=2.75, radius=1.25, shapeName=dyn_hoodDownHair_l_01_capsule, center=<2.7126 0 0>, rotation=<rot 90 0 0>}]
 dyn_hoodDownHair_l_MSTR [{shapeType=Capsule, length=0.5, radius=0.5, shapeName=dyn_hoodDownHair_l_MSTR_capsule}]
 dyn_hoodDownHair_r_01 [{shapeType=Capsule, length=2.75, radius=1.25, shapeName=dyn_hoodDownHair_r_01_capsule, center=<2.7126 0 0>, rotation=<rot 90 0 0>}]
 dyn_hoodDownHair_r_MSTR [{shapeType=Capsule, length=0.5, radius=0.5, shapeName=dyn_hoodDownHair_r_MSTR_capsule}]
 head [{shapeType=Capsule, length=3.5, radius=7.75, center=<6.2754 -1.2299 0>, rotation=<rot 90 90 104.7147>}]
 upperarm_l [{shapeType=Capsule, length=24.25, radius=7.255, shapeName=upperarm_l_capsule, center=<13.0874 0 0.2284>, rotation=<rot 89 180 180>}]
 upperarm_r [{shapeType=Capsule, length=24.25, radius=7.255, shapeName=upperarm_l_capsule, center=<-13.0874 0 -0.2284>, rotation=<rot 89 180 0>}]
}
constraints [
 {bone1Name=dyn_hoodDownBack_c_01, bone2Name=dyn_hoodDownBack_c_MSTR, swing1LimitDegrees=15, swing2LimitDegrees=1, twistLimitDegrees=0}
 {
  bone1Name=dyn_hoodDownHairA_l_MSTR
  bone2Name=head
  swing1LimitDegrees=45
  swing1Motion=Free
  swing2LimitDegrees=45
  swing2Motion=Free
  twistLimitDegrees=45
  twistMotion=Free
 }
 {bone1Name=dyn_hoodDownHairA_l_01, bone2Name=dyn_hoodDownHairA_l_MSTR, swing1LimitDegrees=10, swing2LimitDegrees=10, twistLimitDegrees=0}
 {
  bone1Name=dyn_hoodDownHairA_r_MSTR
  bone2Name=head
  swing1LimitDegrees=45
  swing1Motion=Free
  swing2LimitDegrees=45
  swing2Motion=Free
  twistLimitDegrees=45
  twistMotion=Free
 }
 {bone1Name=dyn_hoodDownHairA_r_01, bone2Name=dyn_hoodDownHairA_r_MSTR, swing1LimitDegrees=10, swing2LimitDegrees=10, twistLimitDegrees=0}
 {
  bone1Name=dyn_hoodDownHair_l_MSTR
  bone2Name=head
  swing1LimitDegrees=45
  swing1Motion=Free
  swing2LimitDegrees=45
  swing2Motion=Free
  twistLimitDegrees=45
  twistMotion=Free
 }
 {bone1Name=dyn_hoodDownHair_l_01, bone2Name=dyn_hoodDownHair_l_MSTR, swing1LimitDegrees=10, swing2LimitDegrees=10, twistLimitDegrees=0}
 {
  bone1Name=dyn_hoodDownHair_r_MSTR
  bone2Name=head
  swing1LimitDegrees=45
  swing1Motion=Free
  swing2LimitDegrees=45
  swing2Motion=Free
  twistLimitDegrees=45
  twistMotion=Free
 }
 {bone1Name=dyn_hoodDownHair_r_01, bone2Name=dyn_hoodDownHair_r_MSTR, swing1LimitDegrees=10, swing2LimitDegrees=10, twistLimitDegrees=0}
 {bone1Name=dyn_hoodDownEar_l_MSTR, bone2Name=dyn_hoodDownBack_c_01, swing1LimitDegrees=5, swing2LimitDegrees=20, twistLimitDegrees=0}
 {bone1Name=dyn_hoodDownEar_r_MSTR, bone2Name=dyn_hoodDownBack_c_01, swing1LimitDegrees=5, swing2LimitDegrees=20, twistLimitDegrees=0}
 {bone1Name=dyn_hoodDownFront_c_01, bone2Name=dyn_hoodDownFront_c_MSTR, swing1LimitDegrees=20, swing2LimitDegrees=20, twistLimitDegrees=0}
 {bone1Name=dyn_hoodDownBackTop_c_01, bone2Name=dyn_hoodDownBackTop_c_MSTR, swing1LimitDegrees=15, swing2LimitDegrees=5, twistLimitDegrees=0}
 {
  bone1Name=upperarm_l
  bone2Name=clavicle_l
  swing1LimitDegrees=45
  swing1Motion=Free
  swing2LimitDegrees=45
  swing2Motion=Free
  twistLimitDegrees=45
  twistMotion=Free
 }
 {
  bone1Name=upperarm_r
  bone2Name=clavicle_r
  swing1LimitDegrees=45
  swing1Motion=Free
  swing2LimitDegrees=45
  swing2Motion=Free
  twistLimitDegrees=45
  twistMotion=Free
 }
]
