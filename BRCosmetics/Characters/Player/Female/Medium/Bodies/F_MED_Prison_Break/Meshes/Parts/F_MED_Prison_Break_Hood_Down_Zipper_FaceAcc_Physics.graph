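# Physics asset for the hood-down zipper on the face accessory part of the female medium Prison Break player cosmetic.
# It has capsule bodies on the zipper bone, a tiny one on its master bone and a large one on the head.
# A single constraint links the zipper to the master with asymmetric swing limits and a zero twist limit.
# The head capsule is unconstrained, apparently serving as a collision volume for the zipper.
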PhysicsAsset /BRCosmetics/Characters/Player/Female/Medium/Bodies/F_MED_Prison_Break/Meshes/Parts/F_MED_Prison_Break_Hood_Down_Zipper_FaceAcc_Physics
bodies {
 dyn_hoodDownZipper_01 [{
   shapeType=Capsule
   length=4.9308
   radius=1.7566
   shapeName=dyn_hoodDownZipper_01_capsule
   center=<2.7288 0.1188 -1>
   rotation=<rot 0 -89.5832 -89.4414>
  }]
 dyn_hoodDownZipper_MSTR [{shapeType=Capsule, length=0.5, radius=0.5, shapeName=dyn_hoodDownZipper_MSTR_capsule}]
 head [{shapeType=Capsule, length=13, radius=9.75, center=<-6.71 -3.4146 0.2222>, rotation=<rot 85.4108 0 0>}]
}
constraints [{bone1Name=dyn_hoodDownZipper_01, bone2Name=dyn_hoodDownZipper_MSTR, swing1LimitDegrees=45, swing2LimitDegrees=60, twistLimitDegrees=0}]
