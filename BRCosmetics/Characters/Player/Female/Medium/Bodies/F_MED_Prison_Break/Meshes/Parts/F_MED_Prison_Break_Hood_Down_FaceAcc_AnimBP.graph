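# Post-pose AnimBlueprint for the female Prison Break lowered-hood part, adding procedural poses and physics to the hood and zipper.
# Copies the parent pose, blending BodyMesh arm and leg bones over it, then adds hood rotations for head tilt while targeting and for skydive floating.
# A hood RigidBody sim, damped while skydiving, feeds a weighted zipper constraint and second sim.
# EventGraph also computes a gravity override the AnimGraph never reads.
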AnimBlueprint /BRCosmetics/Characters/Player/Female/Medium/Bodies/F_MED_Prison_Break/Meshes/Parts/F_MED_Prison_Break_Hood_Down_FaceAcc_AnimBP
target_skeleton=&/F_MED_Prison_Break_Hood_Down_FaceAcc_Skeleton
graphs [
 AnimGraph {nodes [
   n1 = Layered blend per bone(
    BasePose=Copy Pose From Mesh() {bCopyCurves=true, bUseAttachedParent=true}
    BlendPoses_0=Copy Pose From Mesh(SourceMeshComponent=Get BodyMesh() -> (BodyMesh:Skeletal Mesh Component Object Reference))
   ) {layerSetup [{branchFilters [
       {boneName=upperarm_l}
       {boneName=upperarm_r}
       {boneName=lowerarm_l}
       {boneName=lowerarm_r}
       {boneName=thigh_l}
       {boneName=thigh_r}
       {boneName=calf_l}
       {boneName=calf_r}
      ]}]}
   n2 = Get head Rot Back() -> (head Rot Back:float)
   n3 = "Transform (Modify) Bone - Bone: dyn_hoodDownBack_c_MSTR"(ComponentPose=Local To Component(LocalPose=@n1), Alpha=@n2, Translation="0,0,0", Rotation="0,0,30", Scale="1,1,1") {
    rotationMode=BMM_Additive
    rotationSpace=BCS_BoneSpace
    translationSpace=BCS_BoneSpace
    alphaBoolBlend {blendOption=Linear}
    boneToModify {boneName=dyn_hoodDownBack_c_MSTR}
    rotation {roll=30}
   }
   n4 = "Transform (Modify) Bone - Bone: dyn_hoodDownBackTop_c_MSTR"(ComponentPose=@n3, Alpha=@n2, Translation="0,0,0", Rotation="-10,10,-30", Scale="1,1,1") {
    rotationMode=BMM_Additive
    rotationSpace=BCS_BoneSpace
    translationSpace=BCS_BoneSpace
    alphaBoolBlend {blendOption=Linear}
    boneToModify {boneName=dyn_hoodDownBackTop_c_MSTR}
    rotation=<rot -10 10 -30>
   }
   n5 = Get bIsSkydiveFloating()
   n6 = "Transform (Modify) Bone - Bone: dyn_hoodDownBack_c_MSTR"!(ComponentPose=@n4, bAlphaBoolEnabled:bool=@n5, Translation="0,0,0", Rotation="0,25,0", Scale="1,1,1") -> (Pose) {
    alphaInputType=Bool
    rotationMode=BMM_Additive
    rotationSpace=BCS_BoneSpace
    alphaBoolBlend {blendOption=Linear}
    boneToModify {boneName=dyn_hoodDownBack_c_MSTR}
    rotation {yaw=25}
   }
   n7 = "Transform (Modify) Bone - Bone: dyn_hoodDownBackTop_c_MSTR"!(ComponentPose=@n6, bAlphaBoolEnabled:bool=@n5, Translation="0,0,0", Rotation="0,5,0", Scale="1,1,1") -> (Pose) {
    alphaInputType=Bool
    rotationMode=BMM_Additive
    rotationSpace=BCS_BoneSpace
    alphaBoolBlend {blendOption=Linear}
    boneToModify {boneName=dyn_hoodDownBackTop_c_MSTR}
    rotation {yaw=5}
   }
   n8 = Get linearAccScale() -> (linearAccScale:Vector)
   n9 = RigidBody!(ComponentPose=@n7, Alpha, ExternalForce, ComponentLinearAccScale:Vector=@n8, ComponentLinearVelScale:Vector=@n8) -> (Pose) {
    bClampLinearTranslationLimitToRefPose=true
    bOverrideWorldGravity=true
    overridePhysicsAsset=&/F_MED_Prison_Break_Hood_Down_FaceAcc_Physics
    simulationSpace=BaseBoneSpace
    alphaBoolBlend {blendOption=Linear}
    baseBoneRef {boneName=spine_05}
    componentAppliedLinearAccClamp=<6000 6000 6000>
    componentLinearAccScale=<0.3 0.2 0.2>
    componentLinearVelScale=<0.3 0.2 0.2>
    overrideWorldGravity {z=-980}
   }
   n10 = "Constraint - dyn_hoodDownZipper_MSTR (PPPPPPT)"!(
    ComponentPose=@n9
    Alpha=1
    ConstraintWeights_0=0.261
    ConstraintWeights_1:float=0.293
    ConstraintWeights_2:float=0.35
    ConstraintWeights_3:float=0.045
    ConstraintWeights_4:float=0.018
    ConstraintWeights_5:float=0.027
    ConstraintWeights_6:float=0.006
   ) -> (Pose) {
    alphaBoolBlend {blendOption=Linear}
    boneToModify {boneName=dyn_hoodDownZipper_MSTR}
    constraintSetup [
     {targetBone {boneName=dyn_hoodDownBack_c_01}}
     {targetBone {boneName=spine_05}}
     {targetBone {boneName=dyn_hoodDownBackTop_c_01}}
     {targetBone {boneName=clavicle_r}}
     {targetBone {boneName=dyn_hoodDownFront_c_02}}
     {targetBone {boneName=spine_04}}
     {transformType=Translation, targetBone {boneName=deform_pec_r}}
    ]
    constraintWeights [1,1,1,1,1,1,1]
   }
   n11 = RigidBody!(ComponentPose=@n10, Alpha, ExternalForce) -> (Pose) {
    bOverrideWorldGravity=true
    overridePhysicsAsset=&/F_MED_Prison_Break_Hood_Down_Zipper_FaceAcc_Physics
    simulationSpace=BaseBoneSpace
    alphaBoolBlend {blendOption=Linear}
    baseBoneRef {boneName=neck_01}
    componentAppliedLinearAccClamp=<6000 6000 6000>
    componentLinearAccScale=<0.5 0.3 0.3>
    componentLinearVelScale=<0.5 0.3 0.3>
    overrideWorldGravity {z=-980}
   }
   Output Pose(Result=Component To Local(ComponentPose=@n11)) {layerGroup=None}
  ]}
 EventGraph {nodes [
   n1 = Event BlueprintUpdateAnimation()
   > n2 = Sequence!(execute) -> (then_0, then_1, then_2:Exec, then_3:Exec)
   > Set BodyMesh(BodyMesh=GetOwnerSkeletalMeshComponent())
   n3 = Gravity Override State Function DEPRECATED(
    execute=@n2.then_1
    CharacterPartAnimInstance=Self-Reference()
    GameTimeInSeconds=@n1.DeltaTimeX
    GravityOverrideParams=Get Gravity Override Params()
   )
   > Set Gravity Override(Gravity Override=@n3)
   n4 = Branch(execute=@n2.then_2, Condition=Get bIsTargeting())
   > Set head Rot Back(head Rot Back:float=GetAlphaFromReferenceTransformDelta(SkeletalMeshComponent=Get BodyMesh(), BoneName=head, BaseName=spine_05, InputRange=(X=0,Y=-50), TwistAxis="1,0,0")) -> (Output_Get:float)
   Set head Rot Back(execute=@n4.else, head Rot Back=0)
   n5 = Branch(execute=@n2.then_3, Condition=Get bIsSkyDiving())
   > Set linearAccScale(linearAccScale:Vector="0.15,0.05,0.05") -> (Output_Get:Vector)
   Set linearAccScale(execute=@n5.else, linearAccScale="0.3,0.2,0.2")
  ]}
]
variables [
 BodyMesh {type=SkeletalMeshComponent}
 Gravity Override {type=Vector}
 Gravity Override Params {type=GravityOverrideParamsStruct}
 head Rot Back {type=number}
 linearAccScale {type=Vector}
]
